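# Physics asset for the fifth Birthday back bling cosmetic, simulating jiggle on its candles, flames, number five and slice pieces.
# Each dynamic bone has a small capsule body, arranged in a hierarchy under a central dyn_master body.
# Candles, the five and the slice are constrained to the master, and each flame chains to its candle.
# Tight swing and near-locked twist limits keep motion subtle, with flames allowed the most sway.
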PhysicsAsset /BRCosmetics/Accessories/FORT_Backpacks/Backpack_M_MED_FN_Birthday_5/Meshes/M_MED_FN_Birthday_5_Pack_Physics
bodies {
 dyn_candle01 [{shapeType=Capsule, length=2, radius=1, shapeName=dyn_candle01_capsule, center=<2.7294 0 0>, rotation=<rot 90 0 0>}]
 dyn_candle02 [{shapeType=Capsule, length=2, radius=1, shapeName=dyn_candle02_capsule, center=<2.7294 0 0>, rotation=<rot 90 0 0>}]
 dyn_candle03 [{shapeType=Capsule, length=2, radius=1, shapeName=dyn_candle03_capsule, center=<2.7294 0 0>, rotation=<rot 90 0 0>}]
 dyn_candle04 [{shapeType=Capsule, length=2, radius=1, shapeName=dyn_candle04_capsule, center=<2.7294 0 0>, rotation=<rot 90 0 0>}]
 dyn_candle05 [{shapeType=Capsule, length=2, radius=1, shapeName=dyn_candle05_capsule, center=<2.7294 0 0>, rotation=<rot 90 0 0>}]
 dyn_five [{shapeType=Capsule, length=2, radius=2, shapeName=dyn_five_capsule, center=<3.3892 0 0>, rotation=<rot 90 0 0>}]
 dyn_flame01 [{shapeType=Capsule, length=0.5, radius=0.5, shapeName=dyn_flame01_capsule, center=<0.5 0 0>, rotation=<rot 90 0 0>}]
 dyn_flame02 [{shapeType=Capsule, length=0.5, radius=0.5, shapeName=dyn_flame02_capsule, center=<0.5 0 0>, rotation=<rot 90 0 0>}]
 dyn_flame03 [{shapeType=Capsule, length=0.5, radius=0.5, shapeName=dyn_flame03_capsule, center=<0.5 0 0>, rotation=<rot 90 0 0>}]
 dyn_flame04 [{shapeType=Capsule, length=0.5, radius=0.5, shapeName=dyn_flame04_capsule, center=<0.5 0 0>, rotation=<rot 90 0 0>}]
 dyn_flame05 [{shapeType=Capsule, length=0.5, radius=0.5, shapeName=dyn_flame05_capsule, center=<0.5 0 0>, rotation=<rot 90 0 0>}]
 dyn_master [{shapeType=Capsule, length=2, radius=2.75}]
 dyn_slice [{shapeType=Capsule, length=2, radius=2, shapeName=dyn_slice_capsule, center=<5.7172 0 0>, rotation=<rot 90 90 90>}]
}
constraints [
 {bone1Name=dyn_candle05, bone2Name=dyn_master, swing1LimitDegrees=5, swing2LimitDegrees=5, twistLimitDegrees=1}
 {bone1Name=dyn_flame05, bone2Name=dyn_candle05, swing1LimitDegrees=20, swing2LimitDegrees=10, twistLimitDegrees=1}
 {bone1Name=dyn_candle04, bone2Name=dyn_master, swing1LimitDegrees=7, swing2LimitDegrees=7, twistLimitDegrees=1}
 {bone1Name=dyn_flame04, bone2Name=dyn_candle04, swing1LimitDegrees=20, swing2LimitDegrees=10, twistLimitDegrees=1}
 {bone1Name=dyn_candle01, bone2Name=dyn_master, swing1LimitDegrees=5, swing2LimitDegrees=5, twistLimitDegrees=1}
 {bone1Name=dyn_flame01, bone2Name=dyn_candle01, swing1LimitDegrees=20, swing2LimitDegrees=10, twistLimitDegrees=1}
 {bone1Name=dyn_candle03, bone2Name=dyn_master, swing1LimitDegrees=5, swing2LimitDegrees=5, twistLimitDegrees=1}
 {bone1Name=dyn_flame03, bone2Name=dyn_candle03, swing1LimitDegrees=20, swing2LimitDegrees=10, twistLimitDegrees=1}
 {bone1Name=dyn_candle02, bone2Name=dyn_master, swing1LimitDegrees=8, swing2LimitDegrees=8, twistLimitDegrees=1}
 {bone1Name=dyn_flame02, bone2Name=dyn_candle02, swing1LimitDegrees=20, swing2LimitDegrees=10, twistLimitDegrees=1}
 {bone1Name=dyn_five, bone2Name=dyn_master, swing1LimitDegrees=10, swing2LimitDegrees=10, twistLimitDegrees=1}
 {bone1Name=dyn_slice, bone2Name=dyn_master, swing1LimitDegrees=10, swing2LimitDegrees=2, twistLimitDegrees=1}
]
